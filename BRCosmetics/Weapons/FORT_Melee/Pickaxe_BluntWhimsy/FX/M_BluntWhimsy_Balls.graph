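# M_BluntWhimsy_Balls is a Default Lit material for Niagara mesh particles, in the FX set of the BluntWhimsy pickaxe, likely for its ball meshes.
# It is a minimal parameter-driven material with a tan ColorCharger color, scalar parameters R and M (likely roughness and metallic), and one constant.
# The graph has no textures or math nodes and records no links from these values to material outputs.
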Material /BRCosmetics/Weapons/FORT_Melee/Pickaxe_BluntWhimsy/FX/M_BluntWhimsy_Balls
graph {
 properties {bUsedWithNiagaraMeshParticles=true, pixelDepthOffsetMode=PDOM_Legacy, usedShadingModels=Default Lit}
 nodes [
  Constant() {r=0.4}
  ScalarParameter() {defaultValue=0.2238, parameterName=R}
  ScalarParameter() {defaultValue=1, parameterName=M}
  VectorParameter() {parameterName=ColorCharger, defaultValue=<rgba 0.6753 0.4134 0.1653 1>}
 ]
}
