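# Holographic iridescent material for Niagara mesh particles in the Blunt Whimsy pickaxe effects, using Default Lit shading.
# Remapped Fresnel terms build a view-angle coordinate that samples a thin-film spectrum texture, with desaturation varying by the same value.
# That color is blended over a Fresnel-driven gray, scaled by another Fresnel mask and brightened, giving an angle-dependent sheen without animation.
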Material /BRCosmetics/Weapons/FORT_Melee/Pickaxe_BluntWhimsy/FX/M_Particle_BluntWhimsy_Holo
graph {
 properties {bAutomaticallySetUsageInEditor=false, bUsedWithNiagaraMeshParticles=true, usedShadingModels=Default Lit}
 nodes [
  Constant() {r=-0.1}
  n1 = Fresnel() {exponent=0.3}
  n2 = SmoothStep(Value=@n1) {constMin=0.5}
  n3 = LinearInterpolate(Alpha=SmoothStep(Value=@n1) {constMin=0.4}) {constA=0.4, constB=0.6}
  n4 = Multiply(A=Multiply(A=SmoothStep(Value=@n1) {constMin=0.2}, B=Fresnel() {exponent=0.2})) {constB=1.05}
  n5 = Desaturation(
   None=TextureSample(UVs=@n4) {texture=/Game/Global/Textures/Default/Thinfilm/T_ThinFilm_Spectrum_COLOR}.RGB
   Fraction=LinearInterpolate(Alpha=OneMinus(None=@n4)) {constB=0.7}
  )
  Multiply(A=Multiply(A=LinearInterpolate(Alpha=@n2) {constA=0.2}, B=LinearInterpolate(A=@n3, B=@n5))) {constB=1.5}
 ]
}
