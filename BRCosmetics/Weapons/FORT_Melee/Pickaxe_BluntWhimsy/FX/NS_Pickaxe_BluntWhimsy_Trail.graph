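# Niagara swing trail for the BluntWhimsy pickaxe, with six looks selected by User.Variant.
# FN_AnimTrail builds short-lived ribbons from animation trail bone or socket pairs, drawn through ten ribbon renderers.
# Particles_Swing spawns drifting sprites around socket-reader locations, scaled to zero size for variants 0, 1 and 3.
# Variant-indexed arrays set color, trail width and sprite scale, while time-of-day scalars adjust brightness and ribbon opacity.
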NiagaraSystem /BRCosmetics/Weapons/FORT_Melee/Pickaxe_BluntWhimsy/FX/NS_Pickaxe_BluntWhimsy_Trail
dependencies {
 usedDataInterfaces [
  &NiagaraDataInterfaceArrayFloat
  &NiagaraDataInterfaceArrayColor
  &NiagaraDataInterfaceCurve
  &NiagaraDataInterfaceColorCurve
  &NiagaraDataInterfaceSocketReader
 ]
 usedDynamicInputs [&SetBoolByIntComparison, &MergeHelper_FloatSwitch]
 usedModules [
  &self:SystemSpawnScript.SystemScriptSource.SystemScriptGraph.NiagaraNodeAssignment_1.SetVariables_#1
  &self:SystemSpawnScript.SystemScriptSource.SystemScriptGraph.NiagaraNodeAssignment_7.SetVariables_#2
  &self:SystemSpawnScript.SystemScriptSource.SystemScriptGraph.NiagaraNodeAssignment_6.SetVariables_#3
  &self:SystemSpawnScript.SystemScriptSource.SystemScriptGraph.NiagaraNodeAssignment_3.SetVariables_#4
  &self:SystemSpawnScript.SystemScriptSource.SystemScriptGraph.NiagaraNodeAssignment_4.SetVariables_#5
  &self:SystemSpawnScript.SystemScriptSource.SystemScriptGraph.NiagaraNodeAssignment_2.SetVariables_#6
  &self:SystemSpawnScript.SystemScriptSource.SystemScriptGraph.NiagaraNodeAssignment_5.SetVariables_#7
  &SystemState
  &self:SystemSpawnScript.SystemScriptSource.SystemScriptGraph.NiagaraNodeAssignment_0.SetVariables_#8
 ]
}
emitters {
 FN_AnimTrail {
  emitterSpawnScript [&FN_RibbonIDOffsetUtility()]
  emitterUpdateScript [
   &EmitterState()
   &FN_SpawnAnimTrail()
  ]
  particleSpawnScript [
   &FN_InitializeAnimTrailBoneIndecies()
   &FN_InitializeAnimTrailBoneIndecies(AnimTrail Selection=1, First Bone or Socket Index=2, Second Bone or Socket Index=3) {enabled=false, module=FN_InitializeAnimTrailBoneIndecies001}
   &FN_InitializeAnimTrail(AnimTrailWidth=1.75, Lifetime=0.17)
   &AddVelocity(Velocity=<0 400 0>) {enabled=false}
  ]
  particleUpdateScript [
   &SolveForcesAndVelocity(Clamp Velocity=false, Limit Acceleration=false)
   /SetVariables(
    Particles.ColdContrast_Color:LinearColor=<rgba 0.1025 0.06 0.0002 3>
    Particles.GumBall_TrailColor:LinearColor=&Multiply_LinearColorByFloat(
     Linear Color=&SelectLinearColorFromArray(
      Array Sampling Mode=::NewEnumerator1
      Direct Array Index=&MakeIntFromFloat(Float=$User.Variant)
      Color Selection Array=DataInterface() {propertyValues {ColorData [
         <rgba 0.7129 0.4499 1 0.7>
         <rgba 0.9289 0.2409 1 0.7>
         <rgba 0.947 1 0.9798 0.75>
         <rgba 0.947 1 0.9798 0.7>
         <rgba 0.5931 0.3016 0.0391 0.7>
         <rgba 0.947 1 0.9798 0.7>
        ]}}
      Random Seed=?
     )
     Float=&FloatFromCurve(
      FloatCurve=DataInterface() {propertyValues {Curve {keys [<at 0 1>,<at 1 0>]}}}
      CurveIndex=&ReturnNormalizedExecIndex()
      Scale Curve=&FNTimeOfDayScalar(Night Scale=0.7, Morning Scale=1, Evening Scale=0.7)
     )
    )
    Particles.GumBall_TrailWidth:NiagaraFloat=&SelectFloatFromArray(
     Direct Array Index=&MakeIntFromFloat(Float=$User.Variant)
     Float Selection Array=DataInterface() {propertyValues {FloatData [23,23,23,23,23,27]}}
    )
   ) {isSetParameters=true}
   &FN_UpdateAnimTrail() {module=FN_UpdateAnimTrail001}
   &ScaleColor(
    Scale Alpha=&FloatFromCurve(FloatCurve=DataInterface() {propertyValues {Curve {keys [<at 0 1>,<at 1 0>]}}})
    Scale RGB=&MakeVectorFromLinearColorRGB(Color=&MergeHelper_ColorSwitch(
      0=&ColorFromCurve(ColorCurve=DataInterface() {propertyValues {
         AlphaCurve {keys [<at 0 1 RCIM_Cubic>,<at 1 1 RCIM_Cubic>]}
         BlueCurve {keys [<at 0 1.3 RCIM_Cubic>,<at 0.149 1.3>,<at 0.4187 1.3>,<at 0.7148 1.3>,<at 1 1.074 RCIM_Cubic>]}
         GreenCurve {keys [<at 0 0 RCIM_Cubic>,<at 0.149 0.4298>,<at 0.4187 0>,<at 0.7148 0.088>,<at 1 1.3 RCIM_Cubic>]}
         RedCurve {keys [<at 0 0.2227 RCIM_Cubic>,<at 0.149 0.3159>,<at 0.4187 0.3662>,<at 0.7148 0.9704>,<at 1 0.3981 RCIM_Cubic>]}
        }})
      1=<rgba 1.3 0.0676 0.0474 1>
      2=&ColorFromCurve(ColorCurve=DataInterface() {propertyValues {
         AlphaCurve {keys [<at 0 1 RCIM_Cubic>,<at 1 1 RCIM_Cubic>]}
         BlueCurve {keys [<at 0 1.3 RCIM_Cubic>,<at 0.149 1.3>,<at 0.4187 1.3>,<at 0.7148 1.3>,<at 1 1.074 RCIM_Cubic>]}
         GreenCurve {keys [<at 0 0 RCIM_Cubic>,<at 0.149 0.4298>,<at 0.4187 0>,<at 0.7148 0.088>,<at 1 1.3 RCIM_Cubic>]}
         RedCurve {keys [<at 0 0.2227 RCIM_Cubic>,<at 0.149 0.3159>,<at 0.4187 0.3662>,<at 0.7148 0.9704>,<at 1 0.3981 RCIM_Cubic>]}
        }})
      3=&ColorFromCurve(
       ColorCurve=DataInterface() {propertyValues {
         AlphaCurve {keys [<at 0 1 RCIM_Cubic>,<at 1 1 RCIM_Cubic>]}
         BlueCurve {keys [<at 0 1.3 RCIM_Cubic>,<at 0.149 1.3>,<at 0.4187 1.3>,<at 0.7148 1.3>,<at 1 1.074 RCIM_Cubic>]}
         GreenCurve {keys [<at 0 1.1463 RCIM_Cubic>,<at 0.149 0.4298>,<at 0.4187 0.4469>,<at 0.7148 0.088>,<at 1 1.3 RCIM_Cubic>]}
         RedCurve {keys [<at 0 0.4775 RCIM_Cubic>,<at 0.149 0.3159>,<at 0.4187 0.6617>,<at 0.7148 0.9704>,<at 1 0.3981 RCIM_Cubic>]}
        }}
       CurveIndex=&Divide_Float(A=$Particles.NormalizedAge, B=1)
      )
      4=&ColorFromCurve(ColorCurve=DataInterface() {propertyValues {
         AlphaCurve {keys [<at 0 1 RCIM_Cubic>,<at 1 1 RCIM_Cubic>]}
         BlueCurve {keys [<at 0 1.3 RCIM_Cubic>,<at 1 0.8412 RCIM_Cubic>]}
         GreenCurve {keys [<at 0 0.9415 RCIM_Cubic>,<at 1 1.3 RCIM_Cubic>]}
         RedCurve {keys [<at 0 0.0432 RCIM_Cubic>,<at 1 0.0017 RCIM_Cubic>]}
        }})
      5=&ColorFromCurve(ColorCurve=DataInterface() {propertyValues {
         AlphaCurve {keys [<at 0 1 RCIM_Cubic>,<at 1 1 RCIM_Cubic>]}
         BlueCurve {keys [<at 0 0.5252 RCIM_Cubic>,<at 0.514 0.7031 RCIM_Cubic>]}
         GreenCurve {keys [<at 0 0.0477 RCIM_Cubic>,<at 0.514 0.5544 RCIM_Cubic>]}
         RedCurve {keys [<at 0 0.5729 RCIM_Cubic>,<at 0.514 0.1023 RCIM_Cubic>]}
        }})
      Index=&MakeIntFromFloat(Float=$User.Variant, Conversion Mode=::NewEnumerator1)
      Num=::NewEnumerator4
     ))
    ScaleA=true
    ScaleRGB=true
   ) {enabled=false}
   &ScaleColor(
    Scale Alpha=&FloatFromCurve(
     FloatCurve=DataInterface() {propertyValues {Curve {
        defaultValue=0
        keys [
         {arriveTangent=9.0725, interpMode=RCIM_Cubic, leaveTangent=9.0725, tangentMode=RCTM_User, time=0, value=0}
         {arriveTangent=-2.5249, interpMode=RCIM_Cubic, leaveTangent=-2.5249, tangentMode=RCTM_User, time=0.1408, value=0.9847}
         {arriveTangent=-0.0405, interpMode=RCIM_Cubic, leaveTangent=-0.0405, tangentMode=RCTM_User, time=1, value=0}
        ]
       }}}
     Scale Curve=&FNTimeOfDayScalar(
      Night Scale=&SelectFloatFromArray(
       Direct Array Index=&MakeIntFromFloat(Float=$User.Variant)
       Float Selection Array=DataInterface() {propertyValues {FloatData [0.7,0.7,0.7,0.7,0.7,0.25]}}
      )
      Morning Scale=&SelectFloatFromArray(
       Direct Array Index=&MakeIntFromFloat(Float=$User.Variant)
       Float Selection Array=DataInterface() {propertyValues {FloatData [0.9,0.9,0.9,0.9,1,0.35]}}
      )
      Day Scale=&SelectFloatFromArray(
       Direct Array Index=&MakeIntFromFloat(Float=$User.Variant)
       Float Selection Array=DataInterface() {propertyValues {FloatData [0.9,0.9,0.9,0.9,1,0.35]}}
      )
      Evening Scale=&SelectFloatFromArray(
       Direct Array Index=&MakeIntFromFloat(Float=$User.Variant)
       Float Selection Array=DataInterface() {propertyValues {FloatData [0.7,0.7,0.7,0.7,1,0.25]}}
      )
     )
    )
    Scale RGB=&Multiply_VectorByFloat(
     Vector=&MakeVectorFromLinearColorRGB(Color=&SelectLinearColorFromArray(
       Array Sampling Mode=::NewEnumerator1
       Direct Array Index=&MakeIntFromFloat(Float=$User.Variant)
       Color Selection Array=DataInterface() {propertyValues {ColorData [
          <rgba 0.2628 0.1732 1 1>
          <rgba 0.8362 0.1061 1 1>
          <rgba 0.0015 0.5695 1 1>
          <rgba 0.6757 0.644 1 1>
          <rgba 0.4516 0.3153 0.0585 1>
          <rgba 0.2776 0.8719 1 1>
         ]}}
       Random Seed=?
      ))
     Float=&FNTimeOfDayScalar(Night Scale=0.65, Morning Scale=1, Evening Scale=0.65)
    )
    ScaleA=true
    ScaleRGB=true
   ) {module=ScaleColor001}
  ]
  rendererClasses [&NiagaraRibbonRendererProperties]
  renderers [
   {rendererClass=&NiagaraRibbonRendererProperties}
   {rendererClass=&NiagaraRibbonRendererProperties, rendererIndex=1}
   {rendererClass=&NiagaraRibbonRendererProperties, rendererIndex=2}
   {rendererClass=&NiagaraRibbonRendererProperties, rendererIndex=3}
   {rendererClass=&NiagaraRibbonRendererProperties, rendererIndex=4}
   {rendererClass=&NiagaraRibbonRendererProperties, rendererIndex=5}
   {rendererClass=&NiagaraRibbonRendererProperties, rendererIndex=6}
   {rendererClass=&NiagaraRibbonRendererProperties, rendererIndex=7}
   {rendererClass=&NiagaraRibbonRendererProperties, rendererIndex=8}
   {rendererClass=&NiagaraRibbonRendererProperties, rendererIndex=9}
  ]
 }
 Particles_Swing {
  emitterUpdateScript [
   &EmitterState(Life Cycle Mode=::NewEnumerator1, Loop Behavior=::NewEnumerator1, Loop Delay=0.05, Loop Duration Mode=::NewEnumerator1, UseLoopDelay=true)
   &SpawnRate(SpawnRate=100, Use Spawn Probability=false)
  ]
  particleSpawnScript [
   &InitializeParticle(
    Lifetime Max=0.5
    Lifetime Min=0.3
    Lifetime Mode=::NewEnumerator1
    Sprite Rotation Mode=::NewEnumerator2
    Sprite Size Mode=::NewEnumerator1
    Uniform Sprite Size Max=7
    Uniform Sprite Size Min=3
    UsePositionOffset=false
    Write Material Random=true
   )
   &SocketLocation(SamplePosOffset=false, Socket Reader=$System.Socket Reader, Socket Sampling Mode=::NewEnumerator2, VelocitySpeedLimit=false)
   &ShapeLocation(
    MaskSpawnGroup=false
    Offset Mode=::NewEnumerator0
    Override Seed=false
    Rotation Axis=<0 1 0>
    Rotation Mode=::NewEnumerator1
    Sphere Radius=20
    UseOwnerScale=true
   )
   &AddVelocity(Velocity=&RandomRangeVector(Minimum=<0 0 20>, Maximum=<0 0 35>))
  ]
  particleUpdateScript [
   &ParticleState()
   &CurlNoiseForce(
    Noise Frequency=300
    Noise Strength=&RandomRangeFloat(Minimum=200, Maximum=300)
    Pan Noise=true
    Pan Noise Field=<0 0 1>
   ) {enabled=false}
   &SolveForcesAndVelocity(Clamp Velocity=false, Limit Acceleration=false)
   &ScaleColor(
    Scale Alpha=&FloatFromCurve(FloatCurve=DataInterface() {propertyValues {Curve {
        defaultValue=1
        keys [
         {arriveTangent=9.0725, interpMode=RCIM_Cubic, leaveTangent=9.0725, tangentMode=RCTM_User, time=0, value=0}
         {arriveTangent=-0.4317, interpMode=RCIM_Cubic, leaveTangent=-0.4317, tangentMode=RCTM_User, time=0.1408, value=0.9847}
         {arriveTangent=-5.9062, interpMode=RCIM_Cubic, leaveTangent=-5.9062, tangentMode=RCTM_User, time=1, value=0}
        ]
       }}})
    Scale RGB=&Multiply_VectorByFloat(
     Vector=&MakeVectorFromLinearColorRGB(Color=&SelectLinearColorFromArray(
       Array Sampling Mode=::NewEnumerator1
       Direct Array Index=&MakeIntFromFloat(Float=$User.Variant)
       Color Selection Array=DataInterface() {propertyValues {ColorData [
          <rgba 0.2628 0.1732 1 1>
          <rgba 0.8362 0.1061 1 1>
          <rgba 1 1 1 1>
          <rgba 0.6757 0.644 1 1>
          <rgba 0.4751 0.3318 0.0615 1>
          <rgba 0.2776 0.8719 1 1>
         ]}}
       Random Seed=?
      ))
     Float=&FNTimeOfDayScalar(Night Scale=0.65, Morning Scale=1, Evening Scale=0.65)
    )
    ScaleA=true
    ScaleRGB=true
   )
   &ScaleSpriteSize(
    Uniform Curve Scale=&SelectFloatFromArray(
     Direct Array Index=&MakeIntFromFloat(Float=$User.Variant)
     Float Selection Array=DataInterface() {propertyValues {FloatData [0,0,1.5,0,2,2]}}
    )
    Uniform Curve Sprite Scale=DataInterface() {propertyValues {Curve {keys [
        {arriveTangent=8.2127, interpMode=RCIM_Cubic, leaveTangent=8.2127, tangentMode=RCTM_User, time=0, value=0}
        {
         arriveTangent=0.0908
         arriveTangentWeight=0.232
         interpMode=RCIM_Cubic
         leaveTangent=0.0908
         leaveTangentWeight=2.2538
         tangentMode=RCTM_User
         time=0.1
         value=1
        }
        {arriveTangent=-0.042, interpMode=RCIM_Cubic, leaveTangent=-0.042, tangentMode=RCTM_User, time=1, value=0}
       ]}}}
   ) {module=ScaleSpriteSize001}
  ]
  rendererClasses [&NiagaraSpriteRendererProperties]
  renderers [
   {rendererClass=&NiagaraSpriteRendererProperties}
   {rendererClass=&NiagaraSpriteRendererProperties, rendererIndex=1}
   {rendererClass=&NiagaraSpriteRendererProperties, rendererIndex=2}
  ]
 }
}
summary {
 systemName=NS_Pickaxe_BluntWhimsy_Trail
 emitters [
  {emitterName=FN_AnimTrail, rendererClasses [&NiagaraRibbonRendererProperties]}
  {emitterName=Particles_Swing, rendererClasses [&NiagaraSpriteRendererProperties]}
 ]
 userVariables [
  User.bIsFrontEnd {type=NiagaraBool}
  User.bIsFrontEndPreview {type=NiagaraBool}
  User.Debug {type=NiagaraBool}
  User.PhysicalSurface {type=NiagaraInt32}
  User.RibbonIDOffset {type=NiagaraInt32}
  User.SwingSectionIndex {type=NiagaraInt32}
  User.Variant {type=NiagaraFloat, defaultValue=0, description=""}
 ]
}
